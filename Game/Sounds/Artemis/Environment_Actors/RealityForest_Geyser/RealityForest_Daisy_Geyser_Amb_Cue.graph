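# Sound cue for the looping ambient sound of the Reality Forest daisy geyser environment actor.
# A looping close-range geyser wave from the Athena ambient set feeds a modulator that varies pitch slightly while keeping volume constant.
# It plays in the outdoor ambient sound class with its own attenuation asset and overrides concurrency to cap simultaneous instances at a small count.
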SoundCue /Game/Sounds/Artemis/Environment_Actors/RealityForest_Geyser/RealityForest_Daisy_Geyser_Amb_Cue
root=Modulator_0
nodes [
 {id=Modulator_0, children [WavePlayer_0], properties {pitchMax=1.1, pitchMin=0.89, volumeMax=1, volumeMin=1}}
 {id=WavePlayer_0, properties {bLooping=true, soundWaveAssetPtr=/Game/Athena/Sounds/Ambient/Geysers/Geyser_AmbientLoop_Close_01}}
]
properties {
 attenuationSettings=&/RealityForest_Daisy_Geyser_Amb_Att
 bOverrideConcurrency=true
 maxDistance=7000
 soundClassObject=&Fort_Ambient_Outdoors_Athena
 volumeMultiplier=1
 attenuationOverrides {attenuationShapeExtents {x=0}}
 concurrencyOverrides {maxCount=3}
}
